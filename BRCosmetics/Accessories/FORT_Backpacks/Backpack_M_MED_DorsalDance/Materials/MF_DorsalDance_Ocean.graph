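# Material function for the DorsalDance back bling that paints an animated ocean with sharks into the emissive of incoming material attributes, adjusting base color and specular.
# UVs built from pre-skinned local position drive panned low and high wave layers, which shape the bobbing waterline and foam.
# Depth-graded color, panning caustics and two flipbook shark silhouettes make up the water.
# High quality levels blend adjacent flipbook frames; lower ones use single frames.
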MaterialFunction /BRCosmetics/Accessories/FORT_Backpacks/Backpack_M_MED_DorsalDance/Materials/MF_DorsalDance_Ocean
graph {nodes [
  Constant()
  n1 = GetMaterialAttributes(None=FunctionInput() {inputName=Material, inputType=FunctionInput_MaterialAttributes}) {attributeGetTypes [#1, #2]}
  n2 = Multiply(A=@n1.BaseColor) {constB=0.1}
  n3 = NamedRerouteUsage()
  FunctionOutput(None=SetMaterialAttributes!(
    MaterialAttributes=@n1.MaterialAttributes
    EmissiveColor=NamedRerouteUsage()
    BaseColor=LinearInterpolate(A=@n1.BaseColor, B=@n2, Alpha=@n3) {constB=0}
    Specular=LinearInterpolate(A=@n1.Specular, Alpha=@n3) {constB=0}
   ) {attributeSetTypes [#3, #1, #2]})
  n4 = MaterialFunctionCall(Float2=ComponentMask(None=NamedRerouteUsage()) {b=true, r=true}) {materialFunction=&BreakOutFloat2Components}
  n5 = MaterialFunctionCall(X=LinearInterpolate(A=OneMinus(None=@n4.R), B=@n4.R, Alpha=NamedRerouteUsage()), Y=Subtract(A=@n4.G, B=NamedRerouteUsage())) {materialFunction=&MakeFloat2}
  n6 = NamedRerouteDeclaration(None=Divide(
    A=NamedRerouteDeclaration(None=@n5.Result) {name=ModifiedHeightUV, nodeColor=<rgba 1 0 0.6562 1>}
    B=ScalarParameter() {defaultValue=10, group=WaterSimulation_Detail, parameterName=WaveSize}
   )) {name=Wave_UV, nodeColor=<rgba 1 0.7031 0 1>}
  n7 = ScalarParameter() {defaultValue=0.4, group=WaterSimulation_Detail, parameterName=WaveLayerSpeed01X}
  n8 = ScalarParameter() {group=WaterSimulation_Detail, parameterName=WaveLayerSpeed01Y}
  n9 = TextureObjectParameter() {
   group=WaterSimulation_Detail
   parameterName=WaveTexture
   samplerType=SAMPLERTYPE_Masks
   texture=/BRCosmetics/Accessories/FORT_Backpacks/Backpack_M_MED_DorsalDance/Textures/T_OceanLowWave
  }
  n10 = TextureSample(UVs=Panner(Coordinate=@n6, Speed=AppendVector(A=@n7, B=@n8)), Tex=@n9) {
   samplerSource=SSM_Wrap_WorldGroupSettings
   samplerType=SAMPLERTYPE_Masks
   texture=/BRCosmetics/Accessories/FORT_Backpacks/Backpack_M_MED_DorsalDance/Textures/T_OceanCaustic
  }
  n11 = ScalarParameter() {defaultValue=2, group=WaterSimulation_Detail, parameterName=WaveHeight01}
  n12 = ScalarParameter() {defaultValue=0.2, group=WaterSimulation_Detail, parameterName=WaveLayerSpeed02X}
  n13 = ScalarParameter() {group=WaterSimulation_Detail, parameterName=WaveLayerSpeed02Y}
  n14 = TextureSample(UVs=Panner(Coordinate=@n6, Speed=AppendVector(A=@n12, B=@n13)), Tex=@n9) {
   samplerSource=SSM_Wrap_WorldGroupSettings
   samplerType=SAMPLERTYPE_Masks
   texture=&T_OceanCaustic
  }
  n15 = ScalarParameter() {defaultValue=3, group=WaterSimulation_Detail, parameterName=WaveHeight02}
  n16 = ScalarParameter() {defaultValue=0.1, group=WaterSimulation_Detail, parameterName=WaveLayerSpeed03X}
  n17 = ScalarParameter() {group=WaterSimulation_Detail, parameterName=WaveLayerSpeed01Y}
  n18 = TextureSample(UVs=Panner(Coordinate=@n6, Speed=AppendVector(A=@n16, B=@n17)), Tex=@n9) {
   samplerSource=SSM_Wrap_WorldGroupSettings
   samplerType=SAMPLERTYPE_Masks
   texture=&T_OceanCaustic
  }
  n19 = Add(
   A=Add(A=Multiply(A=@n10.R, B=@n11), B=Multiply(A=@n14.G, B=@n15))
   B=Multiply(A=@n18.B, B=ScalarParameter() {defaultValue=5, group=WaterSimulation_Detail, parameterName=WaveHeight03})
  )
  n20 = ComponentMask(None=NamedRerouteUsage()) {g=true}
  n21 = Constant() {r=1}
  n22 = Constant()
  n23 = Multiply(A=If(A=@n20, "A > B"=@n21, "A == B"=@n22, "A < B"=@n22), B=If(A=@n20, "A > B"=@n22, "A == B"=@n22, "A < B"=@n21) {constB=0.99})
  NamedRerouteDeclaration(None=LinearInterpolate(A=Multiply(A=@n19, B=@n23), B=NamedRerouteUsage(), Alpha=NamedRerouteUsage())) {name=HeightNoise, nodeColor=<rgba 0 0.5547 1 1>}
  n24 = PreSkinnedPosition()
  NamedRerouteDeclaration(None=FunctionInput(Preview=VertexInterpolator(VS=@n24).PS) {bUsePreviewValueAsDefault=true, inputName=PreskinnedLocalPos, sortPriority=1}) {name=PreskinnedLocalPos, nodeColor=<rgba 0.4062 0 1 1>}
  NamedRerouteDeclaration(None=@n23) {name=V_Height_0_1, nodeColor=<rgba 1 0.9844 0 1>}
  n25 = ComponentMask(None=NamedRerouteUsage()) {g=true}
  n26 = Constant() {r=1}
  n27 = Constant()
  NamedRerouteDeclaration(None=Multiply(A=If(A=@n25, "A > B"=@n26, "A == B"=@n27, "A < B"=@n27) {constB=0.01}, B=If(A=@n25, "A > B"=@n27, "A == B"=@n27, "A < B"=@n26) {constB=0.99})) {name=V_Foam_0_1, nodeColor=<rgba 1 0.9844 0 1>}
  n28 = NamedRerouteUsage()
  n29 = ComponentMask(None=@n28) {g=true}
  n30 = Constant() {r=1}
  n31 = Constant()
  n32 = ComponentMask(None=@n28) {r=true}
  n33 = ComponentMask(None=Multiply(A=NamedRerouteUsage()) {constB=-1}) {r=true}
  n34 = Add(A=@n33) {constB=0.01}
  n35 = Constant() {r=1}
  n36 = Constant()
  n37 = Add(A=@n33) {constB=0.99}
  n38 = Multiply(
   A=Multiply(A=If(A=@n29, "A > B"=@n30, "A == B"=@n31, "A < B"=@n31) {constB=0.01}, B=If(A=@n29, "A > B"=@n31, "A == B"=@n31, "A < B"=@n30) {constB=0.99})
   B=Multiply(
    A=If(A=@n32, B=@n34, "A > B"=@n35, "A == B"=@n36, "A < B"=@n36) {constB=0.01}
    B=If(A=@n32, B=@n37, "A > B"=@n36, "A == B"=@n36, "A < B"=@n35) {constB=0.99}
   )
  )
  n39 = Multiply(
   A=Add(A=Time(), B=ScalarParameter() {defaultValue=0.2, group=WaterSimulation_Detail, parameterName=TileAnimationOffset01})
   B=ScalarParameter() {defaultValue=1, group=WaterSimulation_Detail, parameterName=SharkTileAnimation01}
  ) {constB=2}
  n40 = MaterialFunctionCall(Float2=ComponentMask(None=NamedRerouteUsage()) {b=true, r=true}) {materialFunction=&BreakOutFloat2Components}
  n41 = OneMinus(None=@n40.R)
  n42 = NamedRerouteUsage()
  n43 = NamedRerouteUsage()
  n44 = Sine(None=Multiply(A=Time(), B=ScalarParameter() {defaultValue=0.3, group=WaterSimulation_Detail, parameterName=SharkAnimationSpeed01}))
  n45 = Add(
   A=Subtract(A=Subtract(A=@n40.G, B=@n43), B=ScalarParameter() {defaultValue=-9, group=WaterSimulation_Detail, parameterName=SharkHeight01})
   B=Multiply(A=ConstantBiasScale(None=@n44), B=ScalarParameter() {defaultValue=6, group=WaterSimulation_Detail, parameterName=SharkSinHeight01})
  )
  n46 = Divide(
   A=MaterialFunctionCall(X=LinearInterpolate(A=@n41, B=@n40.R, Alpha=@n42), Y=OneMinus(None=@n45)) {materialFunction=&MakeFloat2}.Result
   B=ScalarParameter() {defaultValue=12, group=WaterSimulation_Detail, parameterName=SharkSize}
  )
  n47 = Multiply(A=NamedRerouteUsage(), B=Time())
  n48 = ScalarParameter() {group=WaterSimulation_Detail, parameterName=StartPoint01}
  n49 = Constant()
  n50 = Multiply(
   A=Subtract(A=Frac(None=@n47), B=AppendVector(A=@n48, B=@n49))
   B=ScalarParameter() {defaultValue=1, group=WaterSimulation_Detail, parameterName=AcrossLength01}
  )
  n51 = Add(A=@n46, B=@n50)
  n52 = TextureSampleParameter2D(UVs=MaterialFunctionCall("Animation  Phase (0-1)"=Frac(None=@n39), Number of Rows=NamedRerouteUsage(), Number of Columns=NamedRerouteUsage(), UVs=@n51) {materialFunction=&FlipBook}.UVs) {
   group=WaterSimulation_Detail
   parameterName=SharkAtlas
   samplerSource=SSM_Wrap_WorldGroupSettings
   samplerType=SAMPLERTYPE_Masks
   texture=/BRCosmetics/Accessories/FORT_Backpacks/Backpack_M_MED_DorsalDance/Textures/T_DorsalDance_Sharks_Emote
  }
  n53 = Multiply(A=NamedRerouteUsage(), B=NamedRerouteUsage())
  n54 = Add(A=@n39, B=Divide(B=@n53) {constB=16})
  n55 = TextureSampleParameter2D(UVs=MaterialFunctionCall("Animation  Phase (0-1)"=Frac(None=@n54), Number of Rows=NamedRerouteUsage(), Number of Columns=NamedRerouteUsage(), UVs=@n51) {materialFunction=&FlipBook}.UVs) {
   group=WaterSimulation_Detail
   parameterName=SharkAtlas
   samplerSource=SSM_Wrap_WorldGroupSettings
   samplerType=SAMPLERTYPE_Masks
   texture=&T_DorsalDance_Sharks_Emote
  }
  n56 = Multiply(A=@n39, B=@n53) {constB=16}
  n57 = LinearInterpolate(A=@n52.R, B=@n55.R, Alpha=SmoothStep(Value=Frac(None=@n56)))
  n58 = MaterialFunctionCall!(
   Low _ PC Low and Med=@n52.R
   "High _ XB1, PS4"=@n57
   "Epic _ PC High, XSX, PS5"=@n57
   --------
   "Mobile Low _ PC Perf, Android"=@n52.R
   Mobile Med _ Switch=@n52.R
  ) {materialFunction=/Game/Global/MaterialFunctions/TechArt/MF_Fortnite_QualitySwitch_Scalar}
  n59 = NamedRerouteUsage()
  n60 = Multiply(A=Multiply(A=Multiply(A=@n38, B=@n58.Result), B=OneMinus(None=@n59)), B=NamedRerouteUsage())
  n61 = NamedRerouteUsage()
  n62 = ComponentMask(None=@n61) {g=true}
  n63 = Constant() {r=1}
  n64 = Constant()
  n65 = ComponentMask(None=@n61) {r=true}
  n66 = ComponentMask(None=Multiply(A=NamedRerouteUsage()) {constB=-1}) {r=true}
  n67 = Add(A=@n66) {constB=0.01}
  n68 = Constant() {r=1}
  n69 = Constant()
  n70 = Add(A=@n66) {constB=0.99}
  n71 = Multiply(
   A=Multiply(A=If(A=@n62, "A > B"=@n63, "A == B"=@n64, "A < B"=@n64) {constB=0.01}, B=If(A=@n62, "A > B"=@n64, "A == B"=@n64, "A < B"=@n63) {constB=0.99})
   B=Multiply(
    A=If(A=@n65, B=@n67, "A > B"=@n68, "A == B"=@n69, "A < B"=@n69) {constB=0.01}
    B=If(A=@n65, B=@n70, "A > B"=@n69, "A == B"=@n69, "A < B"=@n68) {constB=0.99}
   )
  )
  n72 = Multiply(A=Time(), B=ScalarParameter() {defaultValue=1, group=WaterSimulation_Detail, parameterName=SharkTileAnimation02})
  n73 = ScalarParameter() {defaultValue=0.3, group=WaterSimulation_Detail, parameterName=Xoffset}
  n74 = MaterialFunctionCall(Float2=ComponentMask(None=NamedRerouteUsage()) {b=true, r=true}) {materialFunction=&BreakOutFloat2Components}
  n75 = LinearInterpolate(A=@n74.R, B=OneMinus(None=@n74.R), Alpha=NamedRerouteUsage())
  n76 = NamedRerouteUsage()
  n77 = Sine(None=Multiply(A=Time(), B=ScalarParameter() {defaultValue=0.3, group=WaterSimulation_Detail, parameterName=SharkAnimationSpeed02}))
  n78 = Add(
   A=Subtract(A=Subtract(A=@n74.G, B=@n76), B=ScalarParameter() {defaultValue=-2, group=WaterSimulation_Detail, parameterName=SharkHeight02})
   B=Multiply(A=ConstantBiasScale(None=@n77), B=ScalarParameter() {defaultValue=6, group=WaterSimulation_Detail, parameterName=SharkSinHeight02})
  )
  n79 = Divide(
   A=MaterialFunctionCall(X=Add(A=@n73, B=@n75), Y=OneMinus(None=@n78)) {materialFunction=&MakeFloat2}.Result
   B=ScalarParameter() {defaultValue=12, group=WaterSimulation_Detail, parameterName=SharkSize}
  )
  n80 = Multiply(A=NamedRerouteUsage(), B=Time())
  n81 = ScalarParameter() {defaultValue=0.2, group=WaterSimulation_Detail, parameterName=StartPoint02}
  n82 = Constant()
  n83 = Multiply(
   A=Subtract(A=Frac(None=@n80), B=AppendVector(A=@n81, B=@n82)) {constB=0.2}
   B=ScalarParameter() {defaultValue=1, group=WaterSimulation_Detail, parameterName=AcrossLength02}
  )
  n84 = Add(A=@n79, B=@n83)
  n85 = TextureSampleParameter2D(UVs=MaterialFunctionCall("Animation  Phase (0-1)"=Frac(None=@n72), Number of Rows=NamedRerouteUsage(), Number of Columns=NamedRerouteUsage(), UVs=@n84) {materialFunction=&FlipBook}.UVs) {
   group=WaterSimulation_Detail
   parameterName=SharkAtlas
   samplerSource=SSM_Wrap_WorldGroupSettings
   samplerType=SAMPLERTYPE_Masks
   texture=&T_DorsalDance_Sharks_Emote
  }
  n86 = Multiply(A=NamedRerouteUsage(), B=NamedRerouteUsage())
  n87 = Add(A=@n72, B=Divide(B=@n86) {constB=16})
  n88 = TextureSampleParameter2D(UVs=MaterialFunctionCall("Animation  Phase (0-1)"=Frac(None=@n87), Number of Rows=NamedRerouteUsage(), Number of Columns=NamedRerouteUsage(), UVs=@n84) {materialFunction=&FlipBook}.UVs) {
   group=WaterSimulation_Detail
   parameterName=SharkAtlas
   samplerSource=SSM_Wrap_WorldGroupSettings
   samplerType=SAMPLERTYPE_Masks
   texture=&T_DorsalDance_Sharks_Emote
  }
  n89 = Multiply(A=@n72, B=@n86) {constB=16}
  n90 = LinearInterpolate(A=@n85.R, B=@n88.R, Alpha=SmoothStep(Value=Frac(None=@n89)))
  n91 = MaterialFunctionCall(
   Low _ PC Low and Med=@n85.R
   "High _ XB1, PS4"=@n90
   "Epic _ PC High, XSX, PS5"=@n90
   "Mobile Low _ PC Perf, Android"=@n85.R
   Mobile Med _ Switch=@n85.R
  ) {
   desc=On Highend we blend two frames of flipbook to get smoother result
   materialFunction=&MF_Fortnite_QualitySwitch_Scalar
  }
  n92 = NamedRerouteUsage()
  n93 = Multiply(A=Multiply(A=Multiply(A=@n71, B=@n91.Result), B=OneMinus(None=@n92)), B=NamedRerouteUsage())
  n94 = NamedRerouteUsage()
  n95 = Divide(A=Subtract(A=NamedRerouteUsage(), B=@n94), B=Abs(None=@n94))
  NamedRerouteDeclaration(None=Multiply(A=Add(A=@n60, B=@n93), B=SmoothStep(Value=@n95) {constMax=0.8, constMin=0.6})) {name=SharkIconsAlpha, nodeColor=<rgba 0.5234 1 0 1>}
  n96 = NamedRerouteUsage()
  n97 = VectorParameter() {group=WaterSimulation_Detail, parameterName=PivotOffset, defaultValue=<rgba 0 -4.6 0 1>}
  n98 = Abs(None=ComponentMask(None=Subtract(A=@n96, B=@n97.RGB)) {g=true})
  n99 = Saturate(None=Divide(A=@n98, B=ScalarParameter() {defaultValue=6, group=WaterSimulation_Detail, parameterName=SideMaskSize02}))
  NamedRerouteDeclaration(None=SmoothStep(Value=OneMinus(None=@n99)) {constMax=0.5, constMin=0.2}) {name=SideMask02, nodeColor=<rgba 1 0.1641 0 1>}
  NamedRerouteDeclaration(None=@n46) {name=StaticUV01, nodeColor=<rgba 0 1 0.5078 1>}
  n100 = VectorParameter() {group=WaterSimulation_Detail, parameterName=RowsColumns, defaultValue=<rgba 4 4 4 4>}
  NamedRerouteDeclaration(None=@n100.R) {name=Rows, nodeColor=<rgba 0 0.9297 1 1>}
  NamedRerouteDeclaration(None=@n100.G) {name=Columns, nodeColor=<rgba 1 0 0.1875 1>}
  n101 = MaterialFunctionCall(Float2=ComponentMask(None=NamedRerouteUsage()) {b=true, g=true}) {materialFunction=&BreakOutFloat2Components}
  n102 = LinearInterpolate(A=@n101.R, B=OneMinus(None=@n101.R), Alpha=NamedRerouteUsage())
  n103 = Subtract(A=@n101.G, B=NamedRerouteUsage())
  n104 = NamedRerouteDeclaration(None=Divide(
    A=MaterialFunctionCall(X=@n102, Y=@n103) {materialFunction=&MakeFloat2}.Result
    B=ScalarParameter() {defaultValue=10, group=WaterSimulation_Detail, parameterName=HighWaveSize}
   )) {name=HighWave_UV, nodeColor=<rgba 1 0.7031 0 1>}
  n105 = ScalarParameter() {defaultValue=0.4, group=WaterSimulation_Detail, parameterName=HighWaveLayerSpeed01X}
  n106 = ScalarParameter() {group=WaterSimulation_Detail, parameterName=HighWaveLayerSpeed01Y}
  n107 = TextureObjectParameter() {
   group=WaterSimulation_Detail
   parameterName=WaveTexture01
   samplerType=SAMPLERTYPE_Masks
   texture=/BRCosmetics/Accessories/FORT_Backpacks/Backpack_M_MED_DorsalDance/Textures/T_OceanHighWave
  }
  n108 = TextureSample(UVs=Panner(Coordinate=@n104, Speed=AppendVector(A=@n105, B=@n106)), Tex=@n107) {
   samplerSource=SSM_Wrap_WorldGroupSettings
   samplerType=SAMPLERTYPE_Masks
   texture=&T_OceanCaustic
  }
  n109 = ScalarParameter() {defaultValue=0.2, group=WaterSimulation_Detail, parameterName=HighWaveLayerSpeed02X}
  n110 = ScalarParameter() {group=WaterSimulation_Detail, parameterName=HighWaveLayerSpeed02Y}
  n111 = TextureSample(UVs=Panner(Coordinate=@n104, Speed=AppendVector(A=@n109, B=@n110)), Tex=@n107) {
   samplerSource=SSM_Wrap_WorldGroupSettings
   samplerType=SAMPLERTYPE_Masks
   texture=&T_OceanCaustic
  }
  n112 = Add(
   A=Multiply(A=@n108.R, B=ScalarParameter() {defaultValue=4, group=WaterSimulation_Detail, parameterName=HighWaveHeight01})
   B=Multiply(A=@n111.G, B=ScalarParameter() {defaultValue=5, group=WaterSimulation_Detail, parameterName=HighWaveHeight02})
  )
  n113 = Multiply(B=ScalarParameter() {defaultValue=5, group=WaterSimulation_Detail, parameterName=WaveHeight03})
  n114 = ComponentMask(None=NamedRerouteUsage()) {g=true}
  n115 = Constant() {r=1}
  n116 = Constant()
  n117 = Multiply(A=If(A=@n114, "A > B"=@n115, "A == B"=@n116, "A < B"=@n116), B=If(A=@n114, "A > B"=@n116, "A == B"=@n116, "A < B"=@n115) {constB=0.99})
  NamedRerouteDeclaration(None=Multiply(A=Add(A=@n112, B=@n113), B=NamedRerouteDeclaration(None=@n117) {name=V_Height_0_2, nodeColor=<rgba 1 0.9844 0 1>})) {name=HeightNoise_High, nodeColor=<rgba 0 0.5547 1 1>}
  n118 = NamedRerouteDeclaration(None=Saturate(None=Add(A=@n108.R, B=@n111.G))) {name=HighWaveMask, nodeColor=<rgba 0.7109 1 0 1>}
  n119 = Multiply(A=@n108.R, B=@n111.G)
  n120 = Subtract(A=@n118, B=Saturate(None=NamedRerouteDeclaration(None=@n119) {name=HighWaveOverlay, nodeColor=<rgba 0.0312 0 1 1>}))
  n121 = ComponentMask(None=NamedRerouteUsage()) {g=true}
  n122 = Constant() {r=1}
  n123 = Constant()
  n124 = If(A=@n121, "A > B"=@n122, "A == B"=@n123, "A < B"=@n123) {constB=0.01}
  n125 = If(A=@n121, "A > B"=@n123, "A == B"=@n123, "A < B"=@n122) {constB=0.99}
  NamedRerouteDeclaration(None=Multiply(A=Multiply(A=@n120, B=@n118), B=Multiply(A=@n124, B=@n125))) {name=HighWaveFoam, nodeColor=<rgba 0.2656 1 0 1>}
  n126 = Saturate(None=Divide(A=@n98, B=ScalarParameter() {defaultValue=4, group=WaterSimulation_Detail, parameterName=SideMaskSize01}))
  NamedRerouteDeclaration(None=SmoothStep(Value=OneMinus(None=@n126)) {constMax=0.5}) {name=SideMask01, nodeColor=<rgba 1 0.1641 0 1>}
  NamedRerouteDeclaration(None=@n79) {name=StaticUV02, nodeColor=<rgba 0 1 0.5078 1>}
  n127 = NamedRerouteDeclaration(None=ScalarParameter() {defaultValue=-35, group=WaterSimulation, parameterName=OceanBottom}) {name=SeaFloorHeight, nodeColor=<rgba 1 0.4453 0 1>}
  n128 = Subtract(A=NamedRerouteUsage(), B=Add(A=NamedRerouteUsage(), B=@n127))
  n129 = Multiply(A=@n127) {constB=-1}
  NamedRerouteDeclaration(None=Saturate(None=Divide(A=@n128, B=@n129))) {name=WaterDepthAlpha, nodeColor=<rgba 0.6172 0 1 1>}
  n130 = MaterialFunctionCall(Float2=ComponentMask(None=NamedRerouteUsage()) {b=true, r=true}) {materialFunction=&BreakOutFloat2Components}
  n131 = LinearInterpolate(A=OneMinus(None=@n130.R), B=@n130.R, Alpha=NamedRerouteUsage())
  n132 = ScalarParameter() {defaultValue=10, group=WaterSimulation_Detail, parameterName=CausticsSize}
  n133 = NamedRerouteDeclaration(None=Divide(
    A=MaterialFunctionCall(X=@n131, Y=@n130.G) {materialFunction=&MakeFloat2}.Result
    B=@n132
   )) {name=CausticsUV01, nodeColor=<rgba 0.7109 1 0 1>}
  n134 = TextureObjectParameter() {
   group=WaterSimulation_Detail
   parameterName=CausticTexture
   samplerType=SAMPLERTYPE_Masks
   texture=&T_OceanCaustic
  }
  n135 = TextureSample(UVs=Panner(Coordinate=@n133, Speed=NamedRerouteUsage()), Tex=@n134) {
   samplerSource=SSM_Wrap_WorldGroupSettings
   samplerType=SAMPLERTYPE_Masks
   texture=&T_OceanCaustic
  }
  n136 = MaterialFunctionCall(Float3=@n135.RGB) {materialFunction=&BreakOutFloat3Components}
  n137 = Add(A=@n136.R, B=@n136.R)
  n138 = TextureSample(UVs=Panner(Coordinate=@n133, Speed=NamedRerouteUsage()), Tex=@n134) {
   samplerSource=SSM_Wrap_WorldGroupSettings
   samplerType=SAMPLERTYPE_Masks
   texture=&T_OceanCaustic
  }
  n139 = MaterialFunctionCall(Float3=@n138.RGB) {materialFunction=&BreakOutFloat3Components}
  n140 = Power(Base=Add(A=@n139.G, B=@n139.G))
  n141 = MaterialFunctionCall(Float2=ComponentMask(None=NamedRerouteUsage()) {b=true, g=true}) {materialFunction=&BreakOutFloat2Components}
  n142 = LinearInterpolate(A=@n141.R, B=OneMinus(None=@n141.R), Alpha=NamedRerouteUsage())
  n143 = NamedRerouteDeclaration(None=Divide(
    A=MaterialFunctionCall(X=@n142, Y=@n141.G) {materialFunction=&MakeFloat2}.Result
    B=@n132
   )) {name=CausticsUV02, nodeColor=<rgba 0.7109 1 0 1>}
  n144 = TextureSample(UVs=Panner(Coordinate=@n143, Speed=NamedRerouteUsage()), Tex=@n134) {
   samplerSource=SSM_Wrap_WorldGroupSettings
   samplerType=SAMPLERTYPE_Masks
   texture=&T_OceanCaustic
  }
  n145 = MaterialFunctionCall(Float3=@n144.RGB) {materialFunction=&BreakOutFloat3Components}
  n146 = Add(A=@n145.R, B=@n145.R)
  n147 = TextureSample(UVs=Panner(Coordinate=@n143, Speed=NamedRerouteUsage())) {
   samplerSource=SSM_Wrap_WorldGroupSettings
   samplerType=SAMPLERTYPE_Masks
   texture=&T_OceanCaustic
  }
  n148 = MaterialFunctionCall(Float3=@n147.RGB) {materialFunction=&BreakOutFloat3Components}
  n149 = Power(Base=Add(A=@n148.G, B=@n148.G))
  NamedRerouteDeclaration(None=LinearInterpolate(A=Add(A=@n137, B=@n140), B=Add(A=@n146, B=@n149), Alpha=NamedRerouteUsage())) {name=CausticPattern, nodeColor=<rgba 1 0.8203 0 1>}
  n150 = SmoothStep(Value=NamedRerouteUsage()) {constMin=0.8}
  n151 = LinearInterpolate(
   A=VectorParameter() {group=WaterSimulation, parameterName=OceanDarkenColor, defaultValue=<rgba 0 0.0511 0.1632 1>}.RGB
   B=NamedRerouteDeclaration(None=VectorParameter() {group=WaterSimulation, parameterName=OceanBrightColor, defaultValue=<rgba 0 0.3296 1 1>}.RGB) {name=BrightColor, nodeColor=<rgba 0 0.8125 1 1>}
   Alpha=@n150
  )
  n152 = VectorParameter() {group=WaterSimulation, parameterName=OceanSaturatetColor, defaultValue=<rgba 0.0547 0.1924 0.7812 1>}
  n153 = NamedRerouteUsage()
  n154 = Saturate(None=@n153)
  n155 = Power(Base=Multiply(A=@n150, B=OneMinus(None=@n154)))
  n156 = NamedRerouteUsage()
  n157 = LinearInterpolate(A=NamedRerouteUsage(), Alpha=SmoothStep(Value=@n156) {constMin=0.75})
  n158 = Multiply(
   A=Power(Base=@n153, Exp=ScalarParameter() {defaultValue=2, group=WaterSimulation_Detail, parameterName=CausticPow})
   B=ScalarParameter() {defaultValue=1, group=WaterSimulation_Detail, parameterName=Caustic_EM}
  )
  n159 = Add(
   A=Multiply(
    A=ScalarParameter() {defaultValue=3, group=WaterSimulation, parameterName=WaterEmissiveIntensity}
    B=LinearInterpolate(A=@n151, B=@n152.RGB, Alpha=@n155)
   )
   B=Multiply(A=Multiply(A=@n157, B=@n158), B=SmoothStep(Value=@n156) {constMin=0.65})
  )
  n160 = NamedRerouteUsage()
  n161 = Multiply(
   A=NamedRerouteUsage()
   B=LinearInterpolate(
    A=Add(A=@n159, B=@n160)
    B=VectorParameter() {group=WaterSimulation, parameterName=SharkColor, defaultValue=<rgba 0 0.0584 0.1562 1>}.RGB
    Alpha=NamedRerouteUsage()
   )
  )
  n162 = NamedRerouteUsage()
  NamedRerouteDeclaration(None=Multiply(A=Multiply(A=@n161, B=@n162), B=ScalarParameter() {defaultValue=1, group=WaterSimulation, parameterName=EmissiveIntensityOverall})) {name=EM_OUT, nodeColor=<rgba 1 0 0.75 1>}
  NamedRerouteDeclaration(None=AppendVector(
    A=ScalarParameter() {defaultValue=0.4, group=WaterSimulation_Detail, parameterName=CausticsLayerSpeed01X}
    B=ScalarParameter() {group=WaterSimulation_Detail, parameterName=CausticsLayerSpeed01Y}
   )) {name=CausticsSpeed01, nodeColor=<rgba 0 1 0.6719 1>}
  NamedRerouteDeclaration(None=@n100.B) {name=Rows_1, nodeColor=<rgba 0 0.9297 1 1>}
  NamedRerouteDeclaration(None=AppendVector(
    A=ScalarParameter() {defaultValue=0.2, group=WaterSimulation_Detail, parameterName=CausticsLayerSpeed02X}
    B=ScalarParameter() {group=WaterSimulation_Detail, parameterName=CausticsLayerSpeed02Y}
   )) {name=CausticsSpeed02, nodeColor=<rgba 0 1 0.6719 1>}
  NamedRerouteDeclaration(None=@n100.A) {name=Columns_1, nodeColor=<rgba 1 0 0.1875 1>}
  n163 = VectorParameter() {
   group=WaterSimulation_Detail
   parameterName=SharkSpeed
   channelNames {a=speed02y, b=speed02x, g=speed01y, r=speed01x}
   defaultValue=<rgba 2 0 1 0>
  }
  NamedRerouteDeclaration(None=AppendVector(A=@n163.speed02x, B=@n163.speed02x)) {name=SharkSpeed02, nodeColor=<rgba 0 1 0.5547 1>}
  NamedRerouteDeclaration(None=AppendVector(A=@n163.speed01x, B=@n163.speed01x)) {name=SharkSpeed01, nodeColor=<rgba 0 1 0.5547 1>}
  n164 = NamedRerouteUsage()
  n165 = Sign(None=Subtract(A=ComponentMask(None=@n164) {g=true}, B=ScalarParameter() {defaultValue=-5, group=WaterSimulation_Detail, parameterName=SymetericOffset01}))
  NamedRerouteDeclaration(None=OneMinus(None=Saturate(None=@n165))) {name=FlipMaskG, nodeColor=<rgba 0 1 0.9062 1>}
  n166 = NamedRerouteDeclaration(None=ComponentMask(None=NamedRerouteUsage()) {b=true}) {name=ModifiedHeight, nodeColor=<rgba 0.7109 0 1 1>}
  n167 = ScalarParameter() {group=WaterSimulation, parameterName=WaterSurfaceHeight}
  n168 = Multiply(A=Time(), B=ScalarParameter() {defaultValue=0.4, group=WaterSimulation_Detail, parameterName=WaveOffset_Phase}) {constB=0.2}
  n169 = Multiply(A=ScalarParameter() {defaultValue=1, group=WaterSimulation, parameterName=SinWaveOffsetHeight}, B=Sine(None=Add(A=@n168)))
  n170 = Add(A=NamedRerouteDeclaration(None=Add(A=@n167, B=@n169)) {name=WaterSurfaceHeight, nodeColor=<rgba 1 0 0.8438 1>}, B=NamedRerouteUsage())
  n171 = Constant()
  n172 = Constant() {r=1}
  NamedRerouteDeclaration(None=Multiply(A=If(A=@n166, B=@n170, "A > B"=@n171, "A == B"=@n172, "A < B"=@n172), B=NamedRerouteUsage())) {name=OceanAlpha, nodeColor=<rgba 0.2656 0 1 1>}
  n173 = NamedRerouteUsage()
  n174 = Sign(None=Subtract(A=ComponentMask(None=@n173) {r=true}, B=ScalarParameter() {defaultValue=3, group=WaterSimulation_Detail, parameterName=SymetericOffset02}))
  NamedRerouteDeclaration(None=OneMinus(None=Saturate(None=@n174))) {name=FlipMaskR, nodeColor=<rgba 1 0.8672 0 1>}
  NamedRerouteDeclaration(None=@n50) {name=Offset_01, nodeColor=<rgba 0 0.5312 1 1>}
  NamedRerouteDeclaration(None=@n83) {name=Offset_02, nodeColor=<rgba 0 0.5312 1 1>}
  n175 = TextureSampleParameter2D() {
   parameterName=TechArtMask_RE
   samplerType=SAMPLERTYPE_Masks
   texture=/BRCosmetics/Accessories/FORT_Backpacks/Backpack_M_MED_DorsalDance/Textures/T_DorsalDance_Mask
  }
  NamedRerouteDeclaration(None=@n175.R) {name=FX_Area, nodeColor=<rgba 0.9219 1 0 1>}
  n176 = Add(A=@n10.R, B=@n14.G)
  n177 = NamedRerouteDeclaration(None=Saturate(None=Add(A=@n176, B=@n18.B))) {name=WaveMask, nodeColor=<rgba 0.7109 1 0 1>}
  n178 = Multiply(A=Multiply(A=@n10.R, B=@n14.G), B=@n18.B)
  n179 = Saturate(None=Multiply(A=NamedRerouteDeclaration(None=@n178) {name=WaveOverlay, nodeColor=<rgba 0.0312 0 1 1>}))
  n180 = Multiply(A=Multiply(A=Subtract(A=@n177, B=@n179), B=@n177), B=NamedRerouteUsage())
  n181 = NamedRerouteUsage()
  n182 = NamedRerouteUsage()
  NamedRerouteDeclaration(None=Multiply(
    A=LinearInterpolate(A=@n180, B=@n181, Alpha=@n182)
    B=ScalarParameter() {defaultValue=3, group=WaterSimulation_Detail, parameterName=FoamIntensity}
   )) {name=WaveFoam, nodeColor=<rgba 0.2656 1 0 1>}
  Reroute()
 ]}
