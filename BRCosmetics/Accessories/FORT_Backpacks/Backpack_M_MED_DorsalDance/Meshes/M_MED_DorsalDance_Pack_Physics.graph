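# Physics asset for the DorsalDance medium male backpack mesh, defining bodies for the pack and a three-segment dynamic flag.
# Capsule bodies cover the root, pack, flag base and each flag bone.
# Constraints chain the flag segments to the base with limited swing that widens toward the tip and little or no twist.
# The base connects to the pack with free swing and twist, while the root body has no constraint.
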PhysicsAsset /BRCosmetics/Accessories/FORT_Backpacks/Backpack_M_MED_DorsalDance/Meshes/M_MED_DorsalDance_Pack_Physics
bodies {
 base_dyn_flag_01_c [{shapeType=Capsule, length=1.51, radius=2.5, shapeName=base_dyn_flag_01_c_capsule, rotation=<rot 0 -90 90.0986>}]
 dyn_flag_01_c [{shapeType=Capsule, length=5.1497, radius=2.5862, shapeName=dyn_flag_01_c_capsule, center=<4.0374 0 0>, rotation=<rot 90 0 0>}]
 dyn_flag_02_c [{shapeType=Capsule, length=3.917, radius=2.477, shapeName=dyn_flag_02_c_capsule, center=<4.1685 0 0>, rotation=<rot 90 0 0>}]
 dyn_flag_03_c [{shapeType=Capsule, length=3.3647, radius=2, shapeName=dyn_flag_03_c_capsule, center=<3.219 0 0>, rotation=<rot 90 0 0>}]
 pack [{shapeType=Capsule, length=24.8962, radius=9.7371, shapeName=pack_capsule, center=<0.7757 -3.7998 -5.1034>, rotation=<rot 0.0718 -90 23.4495>}]
 root [{
   shapeType=Capsule
   length=28.7214
   radius=20.8044
   shapeName=root_capsule
   center=<26.197 20.0082 -5.5098>
   rotation=<rot 73.2682 -0.0003 -0.0003>
  }]
}
constraints [
 {bone1Name=dyn_flag_02_c, bone2Name=dyn_flag_01_c, swing1LimitDegrees=60, swing2LimitDegrees=60, twistLimitDegrees=5}
 {bone1Name=dyn_flag_03_c, bone2Name=dyn_flag_02_c, swing1LimitDegrees=70, swing2LimitDegrees=70, twistLimitDegrees=5}
 {bone1Name=dyn_flag_01_c, bone2Name=base_dyn_flag_01_c, swing1LimitDegrees=40, swing2LimitDegrees=40, twistLimitDegrees=0}
 {
  bone1Name=base_dyn_flag_01_c
  bone2Name=pack
  swing1LimitDegrees=45
  swing1Motion=Free
  swing2LimitDegrees=45
  swing2Motion=Free
  twistLimitDegrees=45
  twistMotion=Free
 }
]
