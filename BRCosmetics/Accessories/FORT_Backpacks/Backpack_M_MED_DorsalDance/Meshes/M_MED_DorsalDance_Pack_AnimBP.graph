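# Animation Blueprint for the DorsalDance backpack cosmetic, adding physics-driven secondary motion to its flag bones.
# A cached input pose feeds two RigidBody and bone-offset chains, blended on bIsInFrontEnd, with the gameplay chain adding root and flag-base offsets.
# Additive Transform Bone layers pose the flag segments and pack around the simulation.
# The EventGraph detects body type and updates gravity override, movement-state alphas and phase-shifted sine rotators for wave motion.
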AnimBlueprint /BRCosmetics/Accessories/FORT_Backpacks/Backpack_M_MED_DorsalDance/Meshes/M_MED_DorsalDance_Pack_AnimBP
target_skeleton=&/M_MED_DorsalDance_Pack_Skeleton
graphs [
 AnimGraph {nodes [
   n1 = "Transform (Modify) Bone - Bone: root"!(
    ComponentPose=Local To Component(LocalPose=Use cached pose 'CRCache_Input'())
    bAlphaBoolEnabled:bool=True
    Translation="0,2,0"
    Rotation="0,0,0"
   ) -> (Pose) {
    alphaInputType=Bool
    rotationMode=BMM_Additive
    rotationSpace=BCS_BoneSpace
    translationMode=BMM_Additive
    translationSpace=BCS_BoneSpace
    alphaBoolBlend {blendOption=Linear}
    alphaScaleBiasClamp {bInterpResult=true}
    boneToModify {boneName=root}
    translation {y=2}
   }
   n2 = "Transform (Modify) Bone - Bone: base_dyn_flag_rot_c"!(ComponentPose=@n1, bAlphaBoolEnabled:bool=True, Translation="0,0,0", Rotation="0,-30,0") -> (Pose) {
    alphaInputType=Bool
    rotationMode=BMM_Additive
    translationMode=BMM_Additive
    alphaBoolBlend {blendOption=Linear}
    boneToModify {boneName=base_dyn_flag_rot_c}
    rotation {yaw=-30}
   }
   n3 = RigidBody!(ComponentPose=@n2, Alpha, OverrideWorldGravity:Vector="0,0,-980", ExternalForce) -> (Pose) {
    bClampLinearTranslationLimitToRefPose=true
    bOverrideWorldGravity=true
    overridePhysicsAsset=&/M_MED_DorsalDance_Pack_Physics
    simulationSpace=BaseBoneSpace
    alphaBoolBlend {blendOption=Linear}
    baseBoneRef {boneName=root}
    componentAppliedLinearAccClamp=<3000 3000 3000>
    componentLinearAccScale=<0.3 0.3 0.2>
    componentLinearVelScale=<0.3 0.3 0.2>
    simSpaceSettings {
     maxAngularAcceleration=1000
     maxAngularVelocity=300
     maxLinearAcceleration=700
     maxLinearVelocity=3
     velocityScaleZ=0
     worldAlpha=0.7
     externalLinearDragV=<0.2 0.2>
    }
   }
   n4 = "Transform (Modify) Bone - Bone: dyn_flag_01_c"!(ComponentPose=@n3, Alpha=1, Translation="0,0,0", Rotation="0,0,0") -> (Pose) {
    rotationMode=BMM_Additive
    rotationSpace=BCS_BoneSpace
    translationMode=BMM_Additive
    translationSpace=BCS_BoneSpace
    alphaBoolBlend {blendOption=Linear}
    alphaScaleBiasClamp {bInterpResult=true}
    boneToModify {boneName=dyn_flag_01_c}
   }
   n5 = "Transform (Modify) Bone - Bone: dyn_flag_02_c"!(ComponentPose=@n4, Alpha=1, Translation="0,0,0", Rotation="0,0,0") -> (Pose) {
    rotationMode=BMM_Additive
    rotationSpace=BCS_BoneSpace
    translationMode=BMM_Additive
    translationSpace=BCS_BoneSpace
    alphaBoolBlend {blendOption=Linear}
    alphaScaleBiasClamp {bInterpResult=true}
    boneToModify {boneName=dyn_flag_02_c}
   }
   n6 = "Transform (Modify) Bone - Bone: dyn_flag_03_c"!(ComponentPose=@n5, Alpha=1, Translation="0,0,0", Rotation="0,0,0") -> (Pose) {
    rotationMode=BMM_Additive
    rotationSpace=BCS_BoneSpace
    translationMode=BMM_Additive
    translationSpace=BCS_BoneSpace
    alphaBoolBlend {blendOption=Linear}
    alphaScaleBiasClamp {bInterpResult=true}
    boneToModify {boneName=dyn_flag_03_c}
   }
   n7 = "Transform (Modify) Bone - Bone: dyn_flag_01_c"(ComponentPose=@n6, Alpha=1, Translation="0,0,0", Rotation="0,0,0") {
    rotationMode=BMM_Additive
    rotationSpace=BCS_BoneSpace
    translationMode=BMM_Additive
    translationSpace=BCS_BoneSpace
    alphaBoolBlend {blendOption=Linear}
    alphaScaleBiasClamp {bInterpResult=true}
    boneToModify {boneName=dyn_flag_01_c}
   }
   n8 = "Transform (Modify) Bone - Bone: dyn_flag_02_c"(ComponentPose=@n7, Alpha=1, Translation="0,0,0", Rotation="0,0,0") {
    rotationMode=BMM_Additive
    rotationSpace=BCS_BoneSpace
    translationMode=BMM_Additive
    translationSpace=BCS_BoneSpace
    alphaBoolBlend {blendOption=Linear}
    alphaScaleBiasClamp {bInterpResult=true}
    boneToModify {boneName=dyn_flag_02_c}
   }
   n9 = "Transform (Modify) Bone - Bone: dyn_flag_03_c"(ComponentPose=@n8, Alpha=1, Translation="0,0,0", Rotation="0,0,0") {
    rotationMode=BMM_Additive
    rotationSpace=BCS_BoneSpace
    translationMode=BMM_Additive
    translationSpace=BCS_BoneSpace
    alphaBoolBlend {blendOption=Linear}
    alphaScaleBiasClamp {bInterpResult=true}
    boneToModify {boneName=dyn_flag_03_c}
   }
   n10 = "Transform (Modify) Bone - Bone: base_dyn_flag_01_c"!(ComponentPose=@n9, Alpha=1, Translation="0,0,0", Rotation="0,0,0") -> (Pose) {
    rotationMode=BMM_Additive
    rotationSpace=BCS_BoneSpace
    translationMode=BMM_Additive
    translationSpace=BCS_BoneSpace
    alphaBoolBlend {blendOption=Linear}
    alphaScaleBiasClamp {bInterpResult=true}
    boneToModify {boneName=base_dyn_flag_01_c}
   }
   n11 = "Transform (Modify) Bone - Bone: base_dyn_flag_01_c"(ComponentPose=@n10, Alpha=1, Translation="0,0,0", Rotation="0,0,0") {
    rotationMode=BMM_Additive
    rotationSpace=BCS_BoneSpace
    translationMode=BMM_Additive
    translationSpace=BCS_BoneSpace
    alphaBoolBlend {blendOption=Linear}
    alphaScaleBiasClamp {bInterpResult=true}
    boneToModify {boneName=base_dyn_flag_01_c}
   }
   n12 = "Transform (Modify) Bone - Bone: base_dyn_flag_01_c"(ComponentPose=@n11, Alpha=1, Translation="0,0,0", Rotation="0,0,0") {
    rotationMode=BMM_Additive
    rotationSpace=BCS_BoneSpace
    translationMode=BMM_Additive
    translationSpace=BCS_BoneSpace
    alphaBoolBlend {blendOption=Linear}
    alphaScaleBiasClamp {bInterpResult=true}
    boneToModify {boneName=base_dyn_flag_01_c}
   }
   n13 = "Transform (Modify) Bone - Bone: dyn_flag_rot_c"!(ComponentPose=@n12, Alpha=1, Translation="0,0,0", Rotation="-3,-30,2") -> (Pose) {
    rotationMode=BMM_Additive
    rotationSpace=BCS_BoneSpace
    translationMode=BMM_Additive
    translationSpace=BCS_BoneSpace
    alphaBoolBlend {blendOption=Linear}
    alphaScaleBiasClamp {bInterpResult=true, interpSpeedDecreasing=3}
    boneToModify {boneName=dyn_flag_rot_c}
    rotation=<rot -3 -30 2>
   }
   n14 = "Transform (Modify) Bone - Bone: dyn_flag_01_c"(ComponentPose=@n13, Alpha=1, Translation="0,0,0", Rotation="20,-30,0") {
    rotationMode=BMM_Additive
    rotationSpace=BCS_BoneSpace
    translationMode=BMM_Additive
    translationSpace=BCS_BoneSpace
    alphaBoolBlend {blendOption=Linear}
    alphaScaleBiasClamp {bInterpResult=true, interpSpeedDecreasing=3}
    boneToModify {boneName=dyn_flag_01_c}
    rotation {pitch=20, yaw=-30}
   }
   n15 = "Transform (Modify) Bone - Bone: dyn_flag_01_c"(ComponentPose=@n14, Alpha=1, Translation="0,0,0", Rotation="0,0,0") {
    rotationMode=BMM_Additive
    rotationSpace=BCS_BoneSpace
    translationMode=BMM_Additive
    translationSpace=BCS_BoneSpace
    alphaBoolBlend {blendOption=Linear}
    alphaScaleBiasClamp {bInterpResult=true}
    boneToModify {boneName=dyn_flag_01_c}
   }
   n16 = "Transform (Modify) Bone - Bone: dyn_flag_02_c"(ComponentPose=@n15, Alpha=1, Translation="0,0,0", Rotation="0,0,0") {
    rotationMode=BMM_Additive
    rotationSpace=BCS_BoneSpace
    translationMode=BMM_Additive
    translationSpace=BCS_BoneSpace
    alphaBoolBlend {blendOption=Linear}
    alphaScaleBiasClamp {bInterpResult=true}
    boneToModify {boneName=dyn_flag_02_c}
   }
   n17 = "Transform (Modify) Bone - Bone: dyn_flag_03_c"(ComponentPose=@n16, Alpha=1, Translation="0,0,0", Rotation="0,0,0") {
    rotationMode=BMM_Additive
    rotationSpace=BCS_BoneSpace
    translationMode=BMM_Additive
    translationSpace=BCS_BoneSpace
    alphaBoolBlend {blendOption=Linear}
    alphaScaleBiasClamp {bInterpResult=true}
    boneToModify {boneName=dyn_flag_03_c}
   }
   n18 = "Transform (Modify) Bone - Bone: dyn_flag_01_c"(ComponentPose=@n17, Alpha=1, Translation="0,0,0", Rotation="0,0,0") {
    rotationMode=BMM_Additive
    rotationSpace=BCS_BoneSpace
    translationMode=BMM_Additive
    translationSpace=BCS_BoneSpace
    alphaBoolBlend {blendOption=Linear}
    alphaScaleBiasClamp {bInterpResult=true, interpSpeedDecreasing=3, interpSpeedIncreasing=3}
    boneToModify {boneName=dyn_flag_01_c}
   }
   n19 = "Transform (Modify) Bone - Bone: dyn_flag_02_c"(ComponentPose=@n18, Alpha=1, Translation="0,0,0", Rotation="0,0,0") {
    rotationMode=BMM_Additive
    rotationSpace=BCS_BoneSpace
    translationMode=BMM_Additive
    translationSpace=BCS_BoneSpace
    alphaBoolBlend {blendOption=Linear}
    alphaScaleBiasClamp {bInterpResult=true, interpSpeedDecreasing=3, interpSpeedIncreasing=3}
    boneToModify {boneName=dyn_flag_02_c}
   }
   n20 = "Transform (Modify) Bone - Bone: dyn_flag_03_c"(ComponentPose=@n19, Alpha=1, Translation="0,0,0", Rotation="0,0,0") {
    rotationMode=BMM_Additive
    rotationSpace=BCS_BoneSpace
    translationMode=BMM_Additive
    translationSpace=BCS_BoneSpace
    alphaBoolBlend {blendOption=Linear}
    alphaScaleBiasClamp {bInterpResult=true, interpSpeedDecreasing=3, interpSpeedIncreasing=3}
    boneToModify {boneName=dyn_flag_03_c}
   }
   n21 = "Transform (Modify) Bone - Bone: pack"(ComponentPose=@n20, Translation=Get backpack_translation(), Rotation=Get backpack_rotate()) {rotationMode=BMM_Additive, translationMode=BMM_Additive, alphaBoolBlend {blendOption=Linear}, boneToModify {boneName=pack}}
   n22 = RigidBody(ComponentPose=Local To Component(LocalPose=Use cached pose 'CRCache_Input'()), Alpha=0.7, OverrideWorldGravity="0,0,-980") {
    bClampLinearTranslationLimitToRefPose=true
    bOverrideWorldGravity=true
    overridePhysicsAsset=&M_MED_DorsalDance_Pack_Physics
    simulationSpace=BaseBoneSpace
    alphaBoolBlend {blendOption=Linear}
    baseBoneRef {boneName=root}
    componentAppliedLinearAccClamp=<3000 3000 3000>
    componentLinearAccScale=<0.3 0.3 0.2>
    componentLinearVelScale=<0.3 0.3 0.2>
    simSpaceSettings {
     maxAngularAcceleration=1000
     maxAngularVelocity=300
     maxLinearAcceleration=700
     maxLinearVelocity=3
     velocityScaleZ=0
     worldAlpha=0.7
     externalLinearDragV=<0.2 0.2>
    }
   }
   n23 = "Transform (Modify) Bone - Bone: dyn_flag_rot_c"(ComponentPose=@n22, Alpha=1, Translation="0,0,0", Rotation="-3,-30,2") {
    rotationMode=BMM_Additive
    rotationSpace=BCS_BoneSpace
    translationMode=BMM_Additive
    translationSpace=BCS_BoneSpace
    alphaBoolBlend {blendOption=Linear}
    alphaScaleBiasClamp {bInterpResult=true, interpSpeedDecreasing=3}
    boneToModify {boneName=dyn_flag_rot_c}
    rotation=<rot -3 -30 2>
   }
   n24 = "Transform (Modify) Bone - Bone: dyn_flag_01_c"(ComponentPose=@n23, Alpha=1, Translation="0,0,0", Rotation="20,-30,0") {
    rotationMode=BMM_Additive
    rotationSpace=BCS_BoneSpace
    translationMode=BMM_Additive
    translationSpace=BCS_BoneSpace
    alphaBoolBlend {blendOption=Linear}
    alphaScaleBiasClamp {bInterpResult=true, interpSpeedDecreasing=3}
    boneToModify {boneName=dyn_flag_01_c}
    rotation {pitch=20, yaw=-30}
   }
   n25 = "Transform (Modify) Bone - Bone: dyn_flag_01_c"(ComponentPose=@n24, Alpha=1, Translation="0,0,0", Rotation="0,0,0") {
    rotationMode=BMM_Additive
    rotationSpace=BCS_BoneSpace
    translationMode=BMM_Additive
    translationSpace=BCS_BoneSpace
    alphaBoolBlend {blendOption=Linear}
    alphaScaleBiasClamp {bInterpResult=true}
    boneToModify {boneName=dyn_flag_01_c}
   }
   n26 = "Transform (Modify) Bone - Bone: dyn_flag_02_c"(ComponentPose=@n25, Alpha=1, Translation="0,0,0", Rotation="0,0,0") {
    rotationMode=BMM_Additive
    rotationSpace=BCS_BoneSpace
    translationMode=BMM_Additive
    translationSpace=BCS_BoneSpace
    alphaBoolBlend {blendOption=Linear}
    alphaScaleBiasClamp {bInterpResult=true}
    boneToModify {boneName=dyn_flag_02_c}
   }
   n27 = "Transform (Modify) Bone - Bone: dyn_flag_03_c"(ComponentPose=@n26, Alpha=1, Translation="0,0,0", Rotation="0,0,0") {
    rotationMode=BMM_Additive
    rotationSpace=BCS_BoneSpace
    translationMode=BMM_Additive
    translationSpace=BCS_BoneSpace
    alphaBoolBlend {blendOption=Linear}
    alphaScaleBiasClamp {bInterpResult=true}
    boneToModify {boneName=dyn_flag_03_c}
   }
   n28 = "Transform (Modify) Bone - Bone: dyn_flag_01_c"(ComponentPose=@n27, Alpha=1, Translation="0,0,0", Rotation="0,0,0") {
    rotationMode=BMM_Additive
    rotationSpace=BCS_BoneSpace
    translationMode=BMM_Additive
    translationSpace=BCS_BoneSpace
    alphaBoolBlend {blendOption=Linear}
    alphaScaleBiasClamp {bInterpResult=true, interpSpeedDecreasing=3, interpSpeedIncreasing=3}
    boneToModify {boneName=dyn_flag_01_c}
   }
   n29 = "Transform (Modify) Bone - Bone: dyn_flag_02_c"(ComponentPose=@n28, Alpha=1, Translation="0,0,0", Rotation="0,0,0") {
    rotationMode=BMM_Additive
    rotationSpace=BCS_BoneSpace
    translationMode=BMM_Additive
    translationSpace=BCS_BoneSpace
    alphaBoolBlend {blendOption=Linear}
    alphaScaleBiasClamp {bInterpResult=true, interpSpeedDecreasing=3, interpSpeedIncreasing=3}
    boneToModify {boneName=dyn_flag_02_c}
   }
   n30 = "Transform (Modify) Bone - Bone: dyn_flag_03_c"(ComponentPose=@n29, Alpha=1, Translation="0,0,0", Rotation="0,0,0") {
    rotationMode=BMM_Additive
    rotationSpace=BCS_BoneSpace
    translationMode=BMM_Additive
    translationSpace=BCS_BoneSpace
    alphaBoolBlend {blendOption=Linear}
    alphaScaleBiasClamp {bInterpResult=true, interpSpeedDecreasing=3, interpSpeedIncreasing=3}
    boneToModify {boneName=dyn_flag_03_c}
   }
   n31 = "Transform (Modify) Bone - Bone: pack"(ComponentPose=@n30, Translation=Get backpack_translation(), Rotation=Get backpack_rotate()) {rotationMode=BMM_Additive, translationMode=BMM_Additive, alphaBoolBlend {blendOption=Linear}, boneToModify {boneName=pack}}
   Output Pose(Result=Blend Poses by bool(BlendPose_0=Component To Local(ComponentPose=@n31), BlendPose_1=Component To Local(ComponentPose=@n21), bActiveValue=Get bIsInFrontEnd()) {blendType=Linear}) {layerGroup=None}
   Save cached pose 'CRCache_Input'(Pose=Input Pose()) {cacheName=CRCache_Input}
  ]}
 EventGraph {nodes [
   Event OnCharacterCustomizationLoaded()
   > n1 = Set Owning Pawn(Owning Pawn:Pawn Object Reference=TryGetFortPawn()) -> (Output_Get:Pawn Object Reference)
   > Is Valid(InputObject=@n1)
   > n2 = Sequence!(execute) -> (then_0, then_1, then_2:Exec)
   n3 = FindGender!(execute=@n2.then_1, Object:Object Reference=Get Owning Pawn() -> (Owning Pawn:Pawn Object Reference)) -> (True:Exec, False:Exec, OutGender:EFortCustomGender Enum=Invalid)
   > Set isMale(isMale:bool="Equal (Enum)"(A:EFortCustomGender Enum=@n3, B:EFortCustomGender Enum=Male)) -> (Output_Get:bool)
   > Set IsFemale(IsFemale:bool="Equal (Enum)"(A=@n3, B=Female)) -> (Output_Get:bool)
   n4 = FindBodyType!(execute=@n2.then_2, Object:Object Reference=Get Owning Pawn()) -> (True:Exec, False:Exec, OutBodyType:EFortCustomBodyType Enum=NONE)
   > Set isLarge(isLarge:bool="Equal (Enum)"(A:EFortCustomBodyType Enum=@n4, B:EFortCustomBodyType Enum=Large)) -> (Output_Get:bool)
   > Set isMedium(isMedium:bool="Equal (Enum)"(A=@n4, B=Medium)) -> (Output_Get:bool)
   n5 = Cast To FortPlayerPawn(execute=@n2.then_0, Object=Get Owning Pawn())
   > n6 = Set Owning Player Pawn(Owning Player Pawn:Fort Player Pawn Object Reference=@n5) -> (Output_Get:Fort Player Pawn Object Reference)
   > Set Body Mesh(Body Mesh:Skeletal Mesh Component Object Reference=GetSkeletalMeshForPartType(self=@n6)) -> (Output_Get:Skeletal Mesh Component Object Reference)
   Event BlueprintUpdateAnimation()
   > n7 = Sequence!(execute) -> (then_0, then_1, then_2, then_3:Exec)
   n8 = GetRealTimeSeconds()
   n9 = "Sin (Radians)"(A=float * float(A=float * float(A=1.1, B=1.1), B=@n8))
   n10 = "Sin (Radians)"(A=float * float(A=float * float(A=0.9, B=0.5), B=@n8))
   n11 = "Sin (Radians)"(A=float * float(A=@n8, B=float * float(A=2, B=3)))
   n12 = Set WingsRotation(
    execute=@n7.then_1
    WingsRotation:Rotator=MakeRotator(
     Roll=float * float(A=@n9, B=0.3)
     Pitch=float * float(A=@n10, B=0.3)
     Yaw=Remap Float(Input=@n11, OldRange=(X=-1,Y=1), NewRange=(X=-5,Y=5)) -> (Output=~)
    )
   ) -> (Output_Get:Rotator)
   n13 = Event BlueprintProcessAnimInputEvent()
   n14 = Branch(execute=@n12.then, Condition=IsNamed(self=@n13.AnimInputEvent, InEventName=aie.Player.Zipline.Entered))
   > Set OnZipline(OnZipline:bool=1) -> (Output_Get:bool)
   > Set ZiplineAlpha(ZiplineAlpha:double=0) -> (Output_Get:double)
   Branch(
    execute=@n14.else
    Condition=OR Boolean!(A=IsNamed(self=@n13.AnimInputEvent, InEventName=aie.Player.Zipline.Exited), B=Get bIsMoving2D(), C:bool=Get bIsJumping()) -> (ReturnValue)
   )
   > Set OnZipline(OnZipline=false)
   > Set ZiplineAlpha(ZiplineAlpha=0.7)
   n15 = float * float(A=GetRealTimeSeconds(), B=10) {is=title K2Node_PromotableOperator}
   n16 = MultiplyByPi!(Value:double="Sin (Radians)"(A=@n15)) -> (ReturnValue:double)
   Set sineRotate1(execute=@n7.then_2, sineRotate1:Rotator=MakeRotator(Yaw=float * float(A=@n16, B=4))) -> (Output_Get:Rotator)
   n17 = MultiplyByPi(Value="Sin (Radians)"(A=float + float(A=@n15, B=-0.5)))
   > Set sineRotate2(sineRotate2:Rotator=MakeRotator(Yaw=float * float(A=@n17, B=8))) -> (Output_Get:Rotator)
   n18 = MultiplyByPi(Value="Sin (Radians)"(A=float + float(A=@n15, B=-1.5)))
   > Set sineRotate3(sineRotate3:Rotator=MakeRotator(Yaw=float * float(A=@n18, B=16))) -> (Output_Get:Rotator)
   > Set sineRotateRail1(sineRotateRail1:Rotator=MakeRotator(Yaw=float * float(A=@n16, B=4))) -> (Output_Get:Rotator)
   > Set sineRotateRail2(sineRotateRail2:Rotator=MakeRotator(Yaw=float * float(A=@n17, B=4))) -> (Output_Get:Rotator)
   > Set sineRotateRail3(sineRotateRail3:Rotator=MakeRotator(Yaw=float * float(A=@n18, B=8))) -> (Output_Get:Rotator)
   n19 = MultiplyByPi(Value="Sin (Radians)"(A=float + float(A=@n15, B=-1.5)))
   > Set sineRotateLef1(sineRotateLef1:Rotator=MakeRotator(Roll=20, Pitch=20, Yaw=float * float(A=@n19, B=4))) -> (Output_Get:Rotator)
   > Set sineRotateLef2(sineRotateLef2:Rotator=MakeRotator(Roll=10, Pitch=10, Yaw=float * float(A=@n19, B=4))) -> (Output_Get:Rotator)
   > Set sineRotateLef3(sineRotateLef3:Rotator=MakeRotator(Roll=5, Pitch=5, Yaw=float * float(A=@n19, B=6))) -> (Output_Get:Rotator)
   n20 = MultiplyByPi(Value="Sin (Radians)"(A=float + float(A=@n15, B=-1.5)))
   > Set sineRotateRight1(sineRotateRight1:Rotator=MakeRotator(Pitch=20, Yaw=float * float(A=@n20, B=6))) -> (Output_Get:Rotator)
   > Set sineRotateRight2(sineRotateRight2:Rotator=MakeRotator(Pitch=10, Yaw=float * float(A=@n20, B=6))) -> (Output_Get:Rotator)
   > Set sineRotateRight3(sineRotateRight3:Rotator=MakeRotator(Pitch=5, Yaw=float * float(A=@n20, B=8))) -> (Output_Get:Rotator)
   > n21 = PlayerMovement(AnimInstance=Self-Reference(), XDotProductCoefficient=100, YDotProductCoefficient=100, ZDotProductCoefficient=100)
   n22 = BreakVector(InVec=MakeVector(X=@n21, Y=@n21.YMovement, Z=@n21.ZMovement))
   > n23 = Set Left(Left:bool="float < float"(A=@n22, B=-50)) -> (Output_Get:bool)
   > n24 = Set Right(Right:bool="float > float"(A=@n22, B=50)) -> (Output_Get:bool)
   > Set Forward(Forward:bool="float > float"(A=@n22.Y, B=50)) -> (Output_Get:bool)
   > Set Backward(Backward:bool="float < float"(A=@n22.Y, B=-50)) -> (Output_Get:bool)
   > Set IsLeft(IsLeft:double="To Float (Boolean)"(InBool=@n23)) -> (Output_Get:double)
   > Set IsRight(IsRight:double="To Float (Boolean)"(InBool=@n24)) -> (Output_Get:double)
   n25 = Gravity Override State Function DEPRECATED(
    execute=@n7.then_0
    CharacterPartAnimInstance=Self-Reference()
    GameTimeInSeconds=GetGameTimeInSeconds()
    GravityOverrideParams=Get Gravity Override Params()
   )
   > Set GravityOverride(GravityOverride=@n25)
   > Set SprintingAlpha(SprintingAlpha:double="To Float (Boolean)"(InBool=Get bIsTacticalSprinting())) -> (Output_Get:double)
   n26 = Get bIsSkydiveDiveMode()
   n27 = Get bIsSkydiveFloating()
   > Set skydiveAlpha(skydiveAlpha:double="To Float (Boolean)"(InBool=OR Boolean(A=@n27, B=@n26))) -> (Output_Get:double)
   > Set SwingApha(SwingApha:double="To Float (Boolean)"(InBool=Get bIsSwimSprinting() -> (bIsSwimSprinting:bool))) -> (Output_Get:double)
   n28 = Get OnZipline() -> (OnZipline:bool)
   n29 = Get bIsOnGrindRail()
   n30 = Get bIsCrouchSprinting()
   n31 = Get bIsCrouching()
   > Set wingRotation_Alpha(wingRotation_Alpha:double="To Float (Boolean)"(InBool=OR Boolean!(A=@n29, B=@n30, C=@n28, D:bool=@n31) -> (ReturnValue))) -> (Output_Get:double)
   > Set MeleeAttackAlpha(MeleeAttackAlpha:double="To Float (Boolean)"(InBool=Get bIsPlayingMeleeAnim())) -> (Output_Get:double)
   > Set Sprinting_Alpha(Sprinting_Alpha:double="To Float (Boolean)"(InBool=Get bIsSprinting())) -> (Output_Get:double)
   > Set RailAlpha(RailAlpha:double="To Float (Boolean)"(InBool=Get bIsOnGrindRail())) -> (Output_Get:double)
   > Set IsFMed(IsFMed:bool=AND Boolean(A=Get isMedium() -> (isMedium:bool), B=Get IsFemale() -> (IsFemale:bool))) -> (Output_Get:bool)
   Event BlueprintInitializeAnimation()
   > Set Skel_Mesh(Skel_Mesh=GetOwnerSkeletalMeshComponent())
   n32 = GetOwnerSkeletalMeshComponent()
   n33 = BreakRotator(InRot=BreakTransform(InTransform=GetDeltaTransformFromRefPose(self=@n32, BoneName=spine_05, BaseName=spine_01)).Rotation)
   Set backpack_rotate(backpack_rotate=MakeRotator(Roll=MapRangeClamped(Value=@n33.Yaw, InRangeA=0, InRangeB=-20, OutRangeB=20)))
   > Set backpack_translation(backpack_translation=MakeVector(Y=MapRangeClamped(Value=@n33.Yaw, InRangeA=0, InRangeB=0, OutRangeB=0)))
   n34 = Get bIsSkyDiving()
   > n35 = Branch(Condition=AND Boolean(A=Get bIsJumping(), B=NOT Boolean(A=@n34)))
   > Set deltaonOff(deltaonOff:float=0) -> (Output_Get:float)
   Set deltaonOff(execute=@n35.else, deltaonOff=0)
   TryGetPawnOwner()
   Get bIsInFrontEnd()
  ]}
]
variables [
 backpack_rotate {type=Rotator}
 backpack_translation {type=Vector}
 Backward {type=boolean}
 Body Mesh {type=SkeletalMeshComponent}
 deltaonOff {type=number}
 Floatting_Alpha {type=number}
 Forward {type=boolean}
 Gravity Override Params {
  type=GravityOverrideParamsStruct
  default {
   dBNO {jointName=root}
   dBNO_Carried {jointName=root}
   emoteOrMelee {gravityOverride=<0 -450 -500>, windAmplitude=<50 300 100>, windFrequency=<1.5 12 10>}
   front_End {gravityOverride=<0 -100 -500>, windAmplitude=<20 200 100>, windFrequency=<1.5 8 8>}
   onGround_Moving {gravityOverride=<0 -150 -500>, windAmplitude=<50 300 100>, windFrequency=<1.5 12 10>}
   onGround_Standing {gravityOverride=<0 -450 -500>, windAmplitude=<50 300 100>, windFrequency=<1.5 12 10>}
   rail_Grinding {gravityOverride=<0 0 0>}
   skydive_DivingDown {jointName=root}
   swimming {jointName=root}
   ziplining {windAmplitude=<50 300 100>, windFrequency=<1.5 12 10>}
  }
 }
 GravityOverride {type=Vector}
 IsFemale {type=boolean}
 IsFMed {type=boolean}
 isLarge {type=boolean}
 IsLeft {type=number}
 isM_MED {type=boolean}
 isMale {type=boolean}
 isMedium {type=boolean}
 IsRight {type=number}
 Left {type=boolean}
 MeleeAttackAlpha {type=number}
 OnZipline {type=boolean}
 Owning Pawn {type=Pawn, default=null}
 Owning Player Pawn {type=FortPlayerPawn}
 RailAlpha {type=number}
 Right {type=boolean}
 sineRotate1 {type=Rotator}
 sineRotate2 {type=Rotator}
 sineRotate3 {type=Rotator}
 sineRotateLef1 {type=Rotator}
 sineRotateLef2 {type=Rotator}
 sineRotateLef3 {type=Rotator}
 sineRotateRail1 {type=Rotator}
 sineRotateRail2 {type=Rotator}
 sineRotateRail3 {type=Rotator}
 sineRotateRight1 {type=Rotator}
 sineRotateRight2 {type=Rotator}
 sineRotateRight3 {type=Rotator}
 Skel_Mesh {type=SkeletalMeshComponent}
 skydiveAlpha {type=number}
 Sprinting_Alpha {type=number}
 SprintingAlpha {type=number}
 SwingApha {type=number}
 wingRotation_Alpha {type=number}
 WingsRotation {type=Rotator}
 ZiplineAlpha {type=number}
]
